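# Sound cue for a looping exterior prop ambience, apparently for Chonkers flags, reusing the Creative CTF flag loop wave.
# A looping wave player feeds a short random start delay, then a modulator that lowers pitch and boosts volume with slight randomization.
# Overridden attenuation adds logarithmic falloff, occlusion and listener focus, while a concurrency override caps simultaneous instances.
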SoundCue /Game/Sounds/Fort_Prop_Sounds/Exterior/Chonkers_Flagloop_Amb_Cue
root=Modulator_0
nodes [
 {id=Delay_0, children [WavePlayer_0], properties {delayMax=0.2}}
 {id=Modulator_0, children [Delay_0], properties {pitchMax=0.9, pitchMin=0.75, volumeMax=2, volumeMin=1.9}}
 {id=WavePlayer_0, properties {bLooping=true, soundWaveAssetPtr=/Game/Sounds/Creative/Modes/CTF/CTF_FlagLoop_01}}
]
properties {
 attenuationSettings=&Fort_Ambient_Athena_Att
 bOverrideAttenuation=true
 bOverrideConcurrency=true
 maxDistance=2000
 soundClassObject=&Fort_Ambient_Props_Athena
 volumeMultiplier=0.752
 attenuationOverrides {
  bEnableFocusInterpolation=true
  bEnableListenerFocus=true
  bEnableOcclusion=true
  dBAttenuationAtMax=-40
  distanceAlgorithm=Logarithmic
  falloffDistance=2000
  focusAzimuth=15
  focusVolumeAttenuation=0.8
  nonFocusAzimuth=45
  nonFocusVolumeAttenuation=0.4
  nonSpatializedRadiusStart=100
  occlusionTraceChannel=ECC_GameTraceChannel11
  occlusionVolumeAttenuation=0.4
  stereoSpread=2000
  attenuationShapeExtents {x=0}
 }
 concurrencyOverrides {maxCount=6}
}
