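# Ambient Sound Cue for an exterior windsock prop, built from repurposed flag and chainlink fence wind-rattle recordings.
# A looping flag layer pitched up is mixed with a much quieter, non-looping fence rattle pitched down, each after a short random delay.
# An outer modulator boosts the mix with slight volume variation.
# Overrides add logarithmic falloff, occlusion, listener focus and a concurrency cap.
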SoundCue /Game/Sounds/Fort_Prop_Sounds/Exterior/Chonkers_WindSockLoop_Amb_Cue
root=Modulator_2
nodes [
 {id=Delay_0, children [WavePlayer_0], properties {delayMax=0.2}}
 {id=Delay_1, children [WavePlayer_1], properties {delayMax=0.2}}
 {id=Mixer_0, children [Modulator_0,Modulator_1], properties {inputVolume [1.1,0.7]}}
 {id=Modulator_0, children [Delay_0], properties {pitchMax=1.4, pitchMin=1.2}}
 {id=Modulator_1, children [Delay_1], properties {pitchMax=0.65, pitchMin=0.55, volumeMax=0.1, volumeMin=0.05}}
 {id=Modulator_2, children [Mixer_0], properties {volumeMax=2.35, volumeMin=2.25}}
 {id=WavePlayer_0, properties {bLooping=true, soundWaveAssetPtr=/Game/Sounds/Creative/Modes/CTF/CTF_FlagLoop_01}}
 {
  id=WavePlayer_1
  properties {bLooping=true, soundWaveAssetPtr=/Game/Sounds/Artemis/POI/Hollow_Base/Spots/Spot_Wind_Excited_Chainlink_Fence_Rattle_01}
 }
]
properties {
 attenuationSettings=&Fort_Ambient_Athena_Att
 bOverrideAttenuation=true
 bOverrideConcurrency=true
 maxDistance=2500
 soundClassObject=&Fort_Ambient_Props_Athena
 volumeMultiplier=0.601
 attenuationOverrides {
  bEnableFocusInterpolation=true
  bEnableListenerFocus=true
  bEnableOcclusion=true
  dBAttenuationAtMax=-40
  distanceAlgorithm=Logarithmic
  falloffDistance=2500
  focusAzimuth=15
  focusVolumeAttenuation=1.1
  nonFocusAzimuth=45
  nonFocusVolumeAttenuation=0.4
  nonSpatializedRadiusStart=100
  occlusionTraceChannel=ECC_GameTraceChannel11
  occlusionVolumeAttenuation=0.2
  stereoSpread=2000
  attenuationShapeExtents {x=0}
 }
 concurrencyOverrides {maxCount=6}
}
