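# Sound cue that plays a church bell ring as an exterior prop sound, in the Fort_Ambient_TileMix sound class.
# A single wave player feeds a modulator that slightly varies volume with fixed pitch, then an attenuation node at the root.
# The cue lowers overall volume and overrides concurrency to cap simultaneous instances, preventing new rings once the limit is reached.
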SoundCue /Game/Sounds/Fort_Prop_Sounds/Exterior/Church_Bell_Ring_Cue
root=Attenuation_2
nodes [
 {id=Attenuation_2, children [Modulator_1]}
 {id=Modulator_1, children [WavePlayer_7], properties {pitchMax=1, pitchMin=1, volumeMax=1.1}}
 {id=WavePlayer_7, properties {soundWaveAssetPtr=&/church_bell_ring_01}}
]
properties {
 bOverrideConcurrency=true
 maxDistance=2097152
 soundClassObject=/Game/Sounds/FortSoundClassesAndModes/Fort_Ambient_TileMix
 volumeMultiplier=0.375
 concurrencyOverrides {maxCount=5, resolutionRule=PreventNew}
}
