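# Sound cue for a Halloween door creak, an exterior prop ambience sound in the Fort Athena props sound class.
# A single wave player plays Door_Creak_01 through a modulator that varies playback slightly, with a boosted cue volume.
# It uses the spooky prop attenuation asset and overrides concurrency to allow only one instance, preventing new plays while one is active.
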SoundCue /Game/Sounds/Fort_Prop_Sounds/Exterior/Halloween_Door_Creak_Cue
root=Modulator_0
nodes [
 {id=Modulator_0, children [WavePlayer_1], properties {volumeMax=1}}
 {id=WavePlayer_1, properties {soundWaveAssetPtr=&/Door_Creak_01}}
]
properties {
 attenuationSettings=/Game/Sounds/Attenuation/Fort_Prop_Spooky
 bOverrideConcurrency=true
 maxDistance=2000
 soundClassObject=&Fort_Ambient_Props_Athena
 volumeMultiplier=1.299
 attenuationOverrides {distanceAlgorithm=Logarithmic, falloffDistance=3000}
 concurrencyOverrides {maxCount=1, resolutionRule=PreventNew}
}
