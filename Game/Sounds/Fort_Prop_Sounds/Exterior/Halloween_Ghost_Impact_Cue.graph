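# Sound cue for impacts on the Halloween ghost exterior prop, built from small plastic impact sounds.
# An equally weighted random node picks one of three impact waves and feeds a modulator that varies pitch and volume.
# It plays in the impacts sound class with spooky-prop attenuation.
# Overridden concurrency allows only one instance at a time, blocking new plays while one is active.
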SoundCue /Game/Sounds/Fort_Prop_Sounds/Exterior/Halloween_Ghost_Impact_Cue
root=Modulator_0
nodes [
 {id=Modulator_0, children [Random_1], properties {pitchMax=1, volumeMax=1}}
 {id=Random_1, children [WavePlayer_0,WavePlayer_2,WavePlayer_3], properties {weights [1,1,1]}}
 {id=WavePlayer_0, properties {soundWaveAssetPtr=/Game/Sounds/Fort_Impact_Sounds/Demolition/Plastic/Plastic_Small_Impact01}}
 {id=WavePlayer_2, properties {soundWaveAssetPtr=/Game/Sounds/Fort_Impact_Sounds/Demolition/Plastic/Plastic_Small_Impact02}}
 {id=WavePlayer_3, properties {soundWaveAssetPtr=/Game/Sounds/Fort_Impact_Sounds/Demolition/Plastic/Plastic_Small_Impact03}}
]
properties {
 attenuationSettings=/Game/Sounds/Attenuation/Fort_Prop_Spooky
 bOverrideConcurrency=true
 maxDistance=2000
 soundClassObject=&Fort_Impacts
 volumeMultiplier=0.8
 attenuationOverrides {distanceAlgorithm=Logarithmic, falloffDistance=3000}
 concurrencyOverrides {maxCount=1, resolutionRule=PreventNew}
}
