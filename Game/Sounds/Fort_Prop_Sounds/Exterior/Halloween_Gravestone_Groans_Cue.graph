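# Sound cue for Halloween gravestone prop groans, played as an ambient prop sound with spooky prop attenuation.
# A random node picks one of four equally weighted grave murmur waves, with the fourth boosted in volume by its own modulator.
# A root modulator adds slight pitch and volume variation to the pick.
# Overridden concurrency allows a single instance at a time and blocks new ones while it plays.
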SoundCue /Game/Sounds/Fort_Prop_Sounds/Exterior/Halloween_Gravestone_Groans_Cue
root=Modulator_0
nodes [
 {id=Modulator_0, children [Random_1], properties {pitchMax=1, volumeMax=1}}
 {id=Modulator_1, children [WavePlayer_4], properties {volumeMax=1.3, volumeMin=1.3}}
 {id=Random_1, children [WavePlayer_0,WavePlayer_1,WavePlayer_2,Modulator_1], properties {weights [1,1,1,1]}}
 {id=WavePlayer_0, properties {soundWaveAssetPtr=&/Grave_Murmur_01}}
 {id=WavePlayer_1, properties {soundWaveAssetPtr=&/Grave_Murmur_02}}
 {id=WavePlayer_2, properties {soundWaveAssetPtr=&/Grave_Murmur_03}}
 {id=WavePlayer_4, properties {soundWaveAssetPtr=&/Grave_Murmur_04}}
]
properties {
 attenuationSettings=/Game/Sounds/Attenuation/Fort_Prop_Spooky
 bOverrideConcurrency=true
 maxDistance=2000
 soundClassObject=&Fort_Ambient_Props_Athena
 volumeMultiplier=0.557
 attenuationOverrides {distanceAlgorithm=Logarithmic, falloffDistance=3000}
 concurrencyOverrides {maxCount=1, resolutionRule=PreventNew}
}
