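# Sound cue for Halloween gravestone prop impacts, in the impacts sound class.
# A random node picks one of four equally weighted impact waves, then a modulator varies pitch and volume with both capped at unity.
# It uses spooky prop attenuation, and a concurrency override allows only one instance at a time, blocking new plays rather than cutting off the current one.
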SoundCue /Game/Sounds/Fort_Prop_Sounds/Exterior/Halloween_Gravestone_Impact_Cue
root=Modulator_1
nodes [
 {id=Modulator_1, children [Random_0], properties {pitchMax=1, volumeMax=1}}
 {id=Random_0, children [WavePlayer_0,WavePlayer_2,WavePlayer_3,WavePlayer_4], properties {weights [1,1,1,1]}}
 {id=WavePlayer_0, properties {soundWaveAssetPtr=&/Gravestone_Impact_01}}
 {id=WavePlayer_2, properties {soundWaveAssetPtr=&/Gravestone_Impact_02}}
 {id=WavePlayer_3, properties {soundWaveAssetPtr=&/Gravestone_Impact_03}}
 {id=WavePlayer_4, properties {soundWaveAssetPtr=&/Gravestone_Impact_04}}
]
properties {
 attenuationSettings=/Game/Sounds/Attenuation/Fort_Prop_Spooky
 bOverrideConcurrency=true
 maxDistance=2000
 soundClassObject=&Fort_Impacts
 volumeMultiplier=0.8
 attenuationOverrides {distanceAlgorithm=Logarithmic, falloffDistance=3000}
 concurrencyOverrides {maxCount=1, resolutionRule=PreventNew}
}
