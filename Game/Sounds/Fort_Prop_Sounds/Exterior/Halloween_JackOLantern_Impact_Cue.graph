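# Sound cue for the impact sound of a Halloween jack-o'-lantern exterior prop.
# A mixer layers a quieter tomato toy hit under a random pick from three equally weighted jack-o'-lantern impacts.
# Each layer passes through its own modulator for pitch and volume variation.
# It uses the impacts sound class, spooky prop attenuation, and a concurrency override allowing one instance, blocking new plays while one sounds.
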SoundCue /Game/Sounds/Fort_Prop_Sounds/Exterior/Halloween_JackOLantern_Impact_Cue
root=Mixer_0
nodes [
 {id=Mixer_0, children [Modulator_2,Modulator_0], properties {inputVolume [0.2,1]}}
 {id=Modulator_0, children [Random_1], properties {pitchMax=1, volumeMax=1}}
 {id=Modulator_2, children [WavePlayer_2], properties {pitchMax=1, volumeMax=1}}
 {id=Random_1, children [WavePlayer_3,WavePlayer_4,WavePlayer_5], properties {weights [1,1,1]}}
 {id=WavePlayer_2, properties {soundWaveAssetPtr=/Game/Sounds/Fort_Impact_Sounds/Toys/Toy_Tomato_Hit_02}}
 {id=WavePlayer_3, properties {soundWaveAssetPtr=&/JackOLartern_Impact_01}}
 {id=WavePlayer_4, properties {soundWaveAssetPtr=&/JackOLartern_Impact_02}}
 {id=WavePlayer_5, properties {soundWaveAssetPtr=&/JackOLartern_Impact_03}}
]
properties {
 attenuationSettings=/Game/Sounds/Attenuation/Fort_Prop_Spooky
 bOverrideConcurrency=true
 maxDistance=2000
 soundClassObject=&Fort_Impacts
 volumeMultiplier=0.8
 attenuationOverrides {distanceAlgorithm=Logarithmic, falloffDistance=3000}
 concurrencyOverrides {maxCount=1, resolutionRule=PreventNew}
}
