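# Sound cue for a looping exterior Halloween lawn light prop, reusing a stadium light hum loop.
# A quality level node sends its first two levels through a shared modulator into the looping wave player, leaving the other two levels empty.
# It uses Athena ambient light attenuation and ambient props sound class at reduced volume, with a concurrency override allowing one instance and blocking new ones.
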SoundCue /Game/Sounds/Fort_Prop_Sounds/Exterior/Halloween_Lawn_Light_Loop_Cue
root=QualityLevel_0
nodes [
 {id=Modulator_0, children [WavePlayer_0]}
 {id=QualityLevel_0, children [Modulator_0,Modulator_0,null,null]}
 {id=WavePlayer_0, properties {bLooping=true, soundWaveAssetPtr=&/StadiumLight_Hum_Loop_01}}
]
properties {
 attenuationSettings=/Game/Sounds/Attenuation/Fort_Ambient_Light_Athena_Att
 bOverrideConcurrency=true
 maxDistance=1000
 soundClassObject=&Fort_Ambient_Props_Athena
 volumeMultiplier=0.418
 attenuationOverrides {distanceAlgorithm=Logarithmic, falloffDistance=3000}
 concurrencyOverrides {maxCount=1, resolutionRule=PreventNew}
}
