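# Sound cue for destroying a Halloween spider web prop, routed to the impacts sound class.
# A random node picks one of four equally weighted destroy waves and feeds a modulator for pitch and volume variation.
# The cue lowers pitch, boosts volume and uses spooky-prop attenuation.
# Its concurrency override allows a single instance and blocks new plays while one is active.
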SoundCue /Game/Sounds/Fort_Prop_Sounds/Exterior/Halloween_SpiderWeb_Destroy_Cue
root=Modulator_1
nodes [
 {id=Modulator_1, children [Random_2], properties {pitchMax=1, volumeMax=1}}
 {id=Random_2, children [WavePlayer_1,WavePlayer_5,WavePlayer_6,WavePlayer_7], properties {weights [1,1,1,1]}}
 {id=WavePlayer_1, properties {soundWaveAssetPtr=&/Halloween_SpiderWeb_Destroy_01}}
 {id=WavePlayer_5, properties {soundWaveAssetPtr=&/Halloween_SpiderWeb_Destroy_02}}
 {id=WavePlayer_6, properties {soundWaveAssetPtr=&/Halloween_SpiderWeb_Destroy_03}}
 {id=WavePlayer_7, properties {soundWaveAssetPtr=&/Halloween_SpiderWeb_Destroy_04}}
]
properties {
 attenuationSettings=/Game/Sounds/Attenuation/Fort_Prop_Spooky
 bOverrideConcurrency=true
 maxDistance=2000
 pitchMultiplier=0.8
 soundClassObject=&Fort_Impacts
 volumeMultiplier=2.4
 attenuationOverrides {distanceAlgorithm=Logarithmic, falloffDistance=3000}
 concurrencyOverrides {maxCount=1, resolutionRule=PreventNew}
}
