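# Sound cue for the impact sound of an exterior Halloween witch cackle prop, in the impacts sound class with spooky-prop attenuation.
# A Mixer layers one of three equally weighted balloon pickaxe swing sounds, at reduced volume, over a small plastic impact.
# Each layer passes through its own Modulator for pitch and volume variation.
# A concurrency override allows one instance at a time and blocks new plays while it sounds.
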SoundCue /Game/Sounds/Fort_Prop_Sounds/Exterior/Halloween_Witch_Cackle_Impact_Cue
root=Mixer_0
nodes [
 {id=Mixer_0, children [Modulator_2,Modulator_0], properties {inputVolume [0.4,1]}}
 {id=Modulator_0, children [WavePlayer_8], properties {pitchMax=1, volumeMax=1}}
 {id=Modulator_2, children [Random_0], properties {pitchMax=1, volumeMax=1}}
 {id=Random_0, children [WavePlayer_1, WavePlayer_12, WavePlayer_13], properties {weights [1,1,1]}}
 {id=WavePlayer_1, properties {soundWaveAssetPtr=/Game/Athena/Sounds/Weapons/PickAxes/Balloon/PA_Balloon_Swing_01}}
 {id=WavePlayer_12, properties {soundWaveAssetPtr=/Game/Athena/Sounds/Weapons/PickAxes/Balloon/PA_Balloon_Swing_05}}
 {id=WavePlayer_13, properties {soundWaveAssetPtr=/Game/Athena/Sounds/Weapons/PickAxes/Balloon/PA_Balloon_Swing_06}}
 {id=WavePlayer_8, properties {soundWaveAssetPtr=/Game/Sounds/Fort_Impact_Sounds/Demolition/Plastic/Plastic_Small_Impact02}}
]
properties {
 attenuationSettings=/Game/Sounds/Attenuation/Fort_Prop_Spooky
 bOverrideConcurrency=true
 maxDistance=2000
 soundClassObject=&Fort_Impacts
 volumeMultiplier=0.8
 attenuationOverrides {distanceAlgorithm=Logarithmic, falloffDistance=3000}
 concurrencyOverrides {maxCount=1, resolutionRule=PreventNew}
}
