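# Sound cue for impacts on an exterior lawn light prop.
# A Random node picks one of three equally weighted impact wave variants and plays it slightly boosted through the Fort_Impacts sound class.
# It references the shared Fort_Prop_Spooky attenuation asset for distance falloff.
# A concurrency override allows only one instance at a time and blocks new plays while one is sounding.
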SoundCue /Game/Sounds/Fort_Prop_Sounds/Exterior/Lawn_Light_Impact_Cue
root=Random_0
nodes [
 {id=Random_0, children [WavePlayer_0,WavePlayer_2,WavePlayer_3], properties {weights [1,1,1]}}
 {id=WavePlayer_0, properties {soundWaveAssetPtr=&/Lawn_Light_Impact_01}}
 {id=WavePlayer_2, properties {soundWaveAssetPtr=&/Lawn_Light_Impact_02}}
 {id=WavePlayer_3, properties {soundWaveAssetPtr=&/Lawn_Light_Impact_03}}
]
properties {
 attenuationSettings=/Game/Sounds/Attenuation/Fort_Prop_Spooky
 bOverrideConcurrency=true
 maxDistance=2000
 soundClassObject=&Fort_Impacts
 volumeMultiplier=1.2
 attenuationOverrides {distanceAlgorithm=Logarithmic, falloffDistance=3000}
 concurrencyOverrides {maxCount=1, resolutionRule=PreventNew}
}
