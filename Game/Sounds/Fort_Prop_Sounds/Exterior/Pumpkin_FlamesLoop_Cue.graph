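# Sound cue for the looping flame sound of a pumpkin prop, classed as an Athena ambient prop sound.
# A single looping wave player reuses the creative gadget torch fire loop, pitched up and lowered in volume.
# Inline attenuation overrides the assigned small-toy attenuation asset with a short falloff, low-pass filtering and listener-focus scaling.
# Overridden concurrency allows only two simultaneous instances.
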SoundCue /Game/Sounds/Fort_Prop_Sounds/Exterior/Pumpkin_FlamesLoop_Cue
root=WavePlayer_7
nodes [{id=WavePlayer_7, properties {bLooping=true, soundWaveAssetPtr=/Game/Sounds/Creative/Gadgets/Torch/Torch_Fire_Loop01}}]
properties {
 attenuationSettings=/Game/Sounds/Creative/Attenuations/Fort_Creative_SmallToy_Att
 bOverrideAttenuation=true
 bOverrideConcurrency=true
 maxDistance=500
 pitchMultiplier=1.3
 soundClassObject=&Fort_Ambient_Props_Athena
 volumeMultiplier=0.371
 attenuationOverrides {
  bAttenuateWithLPF=true
  bEnableFocusInterpolation=true
  bEnableListenerFocus=true
  falloffDistance=500
  focusVolumeAttenuation=1.3
  lPFRadiusMin=400
  nonFocusVolumeAttenuation=0.3
  stereoSpread=100
  attenuationShapeExtents {x=0}
 }
 concurrencyOverrides {maxCount=2}
}
