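# Looping ambient hum Sound Cue for a stadium light exterior prop.
# A quality-level root routes its first two slots to one shared looping wave player and leaves the other two empty.
# The cue uses close-range prop attenuation and the Athena ambient props sound class.
# A concurrency override caps simultaneous instances at a low count, and a volume multiplier boosts the level.
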SoundCue /Game/Sounds/Fort_Prop_Sounds/Exterior/StadiumLight_Hum_Loop_Cue
root=QualityLevel_0
nodes [
 {id=QualityLevel_0, children [WavePlayer_0,WavePlayer_0,null,null]}
 {id=WavePlayer_0, properties {bLooping=true, soundWaveAssetPtr=&/StadiumLight_Hum_Loop_01}}
]
properties {
 attenuationSettings=/Game/Sounds/Attenuation/Fort_Prop_Close_Att
 bOverrideConcurrency=true
 maxDistance=1800
 soundClassObject=&Fort_Ambient_Props_Athena
 volumeMultiplier=1.392
 concurrencyOverrides {maxCount=2}
}
